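; Outlander fragment container: an interactable orb that grants Charge or Llama fragment ammo for fragment abilities.
; Interaction checks build mode, ammo capacity and one-Llama limit, and blocks players without the ability while it is lit for a nearby ability-holding Outlander.
; GiveFragment adds ammo with Double Down bonuses, applies type and cooldown effects, and sends a collection event.
; RepNotifies drive minimap color, FX, audio and animation; tick is enabled only for awake, cosmetic-capable instances with an Outlander.
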(fn BlueprintCanInteract
  (params (InteractingPawn :type Fort%20Pawn%20Object%20Reference) (InteractionBeingAttempted :type EInteractionBeingAttempted%20Enum :default "FirstInteraction") (InteractionType :type TInteractionType%20Enum :default "IT_NoInteraction"))
  (bind _returnvalue (Ability|GetAbilitySystemComponent InteractingPawn))
  (bind _tc_llamafragment (Variables|Default|GetTC_LlamaFragment))
  (bind _tc_hasfragmentability (Variables|Default|GetTC_HasFragmentAbility))
  (bind (_case_0_true _case_1_true) (GameplayTags|TagContainer|CompareTagAssetInterfacetoOtherTagContainers _tc_hasfragmentability _tc_llamafragment _returnvalue))
  (bind _fragmentammodata (Variables|Default|GetFragmentAmmoData))
  ;[1 1 1] Can't interact if we're in build mode
  (bind _asfort_player_pawn (Utilities|Casting|CastToFortPlayerPawn InteractingPawn)
    (:then
      (if (Utilities|Enum|Equal(Enum) (Variables|/Script/FortniteGame.FortPlayerPawn|GetBuildingState _asfort_player_pawn) "None")
        (if (CallFunction|IsOutlander InteractingPawn)
          (if _case_0_true
            (bind _itemdata (Class|ItemDefinitionBase|GetItemData _fragmentammodata)
              (:Found
                _)
              (:NotFound
                _))
            (LocalVariables|SetLocalExecuteOk (or (and (or (<= (Inventory|GetItemQuantityOnPlayer (Utilities|Casting|CastToFortPlayerController (Pawn|GetController InteractingPawn)) (Variables|Default|GetFragmentAmmoData)) (- (Math|Integer|SelectInt (Math|Float|Truncate (.ReturnValue (Ability|Attribute|EvaluateAttributeValuewithTagsandBase (Ability|GetAbilitySystemComponent InteractingPawn) (Variables|Default|GetAttributeRegenMaxStack) (GameplayTags|GetOwnedGameplayTags _fragmentammodata :!class /Script/GameplayTags.GameplayTagAssetInterface) (GameplayTags|MakeLiteralGameplayTagContainer) (.ReturnValue (Ability|Attribute|EvaluateAttributeValuewithTagsandBase (Ability|GetAbilitySystemComponent InteractingPawn) (Variables|Default|GetAttributeMaxStack) (GameplayTags|GetOwnedGameplayTags _fragmentammodata :!class /Script/GameplayTags.GameplayTagAssetInterface) (GameplayTags|MakeLiteralGameplayTagContainer) (ScalableFloat|GetScalableFloatMultiplier (Utilities|Struct|BreakMaxStackSizeItemData _itemdata))))))) (Math|Float|Truncate (.ReturnValue (Ability|Attribute|EvaluateAttributeValuewithTagsandBase (Ability|GetAbilitySystemComponent InteractingPawn) (Variables|Default|GetAttributeMaxStack) (GameplayTags|GetOwnedGameplayTags _fragmentammodata :!class /Script/GameplayTags.GameplayTagAssetInterface) (GameplayTags|MakeLiteralGameplayTagContainer) (ScalableFloat|GetScalableFloatMultiplier (Utilities|Struct|BreakMaxStackSizeItemData _itemdata))))) (> (ScalableFloat|GetValueatLevel (Variables|/Script/FortniteGame.FortAmmoItemDefinition|GetRegenCooldown (Variables|Default|GetFragmentAmmoData))) 0.0)) 2)) (and (<= (Inventory|GetItemQuantityOnPlayer (Utilities|Casting|CastToFortPlayerController (Pawn|GetController InteractingPawn)) (Variables|Default|GetFragmentAmmoData)) (- (Math|Integer|SelectInt (Math|Float|Truncate (.ReturnValue (Ability|Attribute|EvaluateAttributeValuewithTagsandBase (Ability|GetAbilitySystemComponent InteractingPawn) (Variables|Default|GetAttributeRegenMaxStack) (GameplayTags|GetOwnedGameplayTags _fragmentammodata :!class /Script/GameplayTags.GameplayTagAssetInterface) (GameplayTags|MakeLiteralGameplayTagContainer) (.ReturnValue (Ability|Attribute|EvaluateAttributeValuewithTagsandBase (Ability|GetAbilitySystemComponent InteractingPawn) (Variables|Default|GetAttributeMaxStack) (GameplayTags|GetOwnedGameplayTags _fragmentammodata :!class /Script/GameplayTags.GameplayTagAssetInterface) (GameplayTags|MakeLiteralGameplayTagContainer) (ScalableFloat|GetScalableFloatMultiplier (Utilities|Struct|BreakMaxStackSizeItemData _itemdata))))))) (Math|Float|Truncate (.ReturnValue (Ability|Attribute|EvaluateAttributeValuewithTagsandBase (Ability|GetAbilitySystemComponent InteractingPawn) (Variables|Default|GetAttributeMaxStack) (GameplayTags|GetOwnedGameplayTags _fragmentammodata :!class /Script/GameplayTags.GameplayTagAssetInterface) (GameplayTags|MakeLiteralGameplayTagContainer) (ScalableFloat|GetScalableFloatMultiplier (Utilities|Struct|BreakMaxStackSizeItemData _itemdata))))) (> (ScalableFloat|GetValueatLevel (Variables|/Script/FortniteGame.FortAmmoItemDefinition|GetRegenCooldown (Variables|Default|GetFragmentAmmoData))) 0.0)) 1)) _case_1_true)) (Utilities|Enum|Equal(Enum) (Variables|Default|GetFragmentType) "NewEnumerator4")) (and (Utilities|Enum|Equal(Enum) (Variables|Default|GetFragmentType) "NewEnumerator2") (not _case_1_true))) :!pinned true) ; Either this is a Charge Fragment, or this is a Llama Fragment and you do not yet have a Llama.\nYou can hold multiple Charge Fragments, but only one Llama Fragment.
            (else
              (LocalVariables|SetLocalExecuteOk (not (Variables|Default|GetIsOn)) :!pinned true :!label _label) ; IsOn means an Outlander with the Fragment Ability is nearby and should be the only one to grab this.
            ))
          (return (and (LocalVariables|GetLocalExecuteOk) (Variables|Default|GetCanInteract))))
        (goto _label)
        (else
          (LocalVariables|SetLocalExecuteOk false :!label _label_2)))
      (return (and (LocalVariables|GetLocalExecuteOk) (Variables|Default|GetCanInteract))))
    (:CastFailed
      _))
  (goto _label_2))

(fn BlueprintGetFailedInteractionString
  (params (InteractingPawn :type Fort%20Pawn%20Object%20Reference) (InteractionBeingAttempted :type EInteractionBeingAttempted%20Enum :default "FirstInteraction"))
  (bind _tc_llamafragment (Variables|Default|GetTC_LlamaFragment))
  (bind _tc_hasfragmentability (Variables|Default|GetTC_HasFragmentAbility))
  (bind (_case_0_true _case_1_true) (GameplayTags|TagContainer|CompareTagAssetInterfacetoOtherTagContainers _tc_hasfragmentability _tc_llamafragment InteractingPawn))
  ;[1 1 1] Don't show failure text if we're in build mode
  (bind _asfort_player_pawn (Utilities|Casting|CastToFortPlayerPawn InteractingPawn)
    (:then
      (if (and (Utilities|Enum|Equal(Enum) (Variables|/Script/FortniteGame.FortPlayerPawn|GetBuildingState _asfort_player_pawn) "None") (and (Variables|Default|GetHasLoot) (> (- (Utilities|Time|GetGameTimeinSeconds) (Variables|Default|GetLastInteractStartTime)) 5)))
        (if (CallFunction|IsOutlander InteractingPawn)
          (if _case_0_true
            (if _case_1_true
              (switch Utilities|FlowControl|Switch|SwitchonE_OutlanderFragmentTypes (Variables|Default|GetFragmentType)
                (:NewEnumerator4
                  _)
                (:NewEnumerator2
                  (return (Variables|DisplayStrings|GetFailedOutlanderOnlyOneLlama)))))
            (return (Variables|DisplayStrings|GetFailedOutlanderMaxFragHeldDefault))
            (else
              (return (Variables|DisplayStrings|GetFailedOutlanderActivatedButDoesNotYetPossessFragmentAbility))))
          (else
            (return (Variables|DisplayStrings|GetFailedNonOutlanderActivated))))
        (else
          _)))
    (:CastFailed
      _))
  (return ""))

(fn BlueprintGetInteractionString
  (params (InteractingPawn :type Fort%20Pawn%20Object%20Reference) (InteractionBeingAttempted :type EInteractionBeingAttempted%20Enum :default "FirstInteraction"))
  (if (CallFunction|IsOutlander InteractingPawn)
    (if (GameplayTags|TagContainer|CompareTagAssetInterfacetoOtherTagContainers (Variables|Default|GetTC_HasFragmentAbility) InteractingPawn)
      (switch Utilities|FlowControl|Switch|SwitchonE_OutlanderFragmentTypes (Variables|Default|GetFragmentType)
        (:NewEnumerator4
          (return (Variables|DisplayStrings|GetInteractChargeFragment)))
        (:NewEnumerator2
          (return (Variables|DisplayStrings|GetInteractLlamaFragment))))
      (else
        _)))
  (return (Variables|DisplayStrings|GetInteractNonOutlander)))

(fn DelayedDestroy
  (Actor|DestroyActor))

(fn ForceFeedbackInteract
  (params (Soft :type Boolean))
  (if Soft
    (bind _asfort_player_controller (Utilities|Casting|CastToFortPlayerController (Pawn|GetController (Variables|Default|GetInteractingPlayerPawn))))
    (Game|Feedback|ClientPlayForceFeedback _asfort_player_controller (Variables|Default|GetFFInteractSoft) "OrbInteractSoft" false false false)
    (else
      (Fortnite|PlayLocalForceFeedbackatLocationMulti (Variables|Default|GetFF_Interact) (Variables|Default|GetFF_Interact) (Transformation|GetWorldLocation (Variables|Default|GetRotationRoot)) 512 1024 "OrbInteract")))
  (return))

(fn GiveFragment
  (bind (_case_0_true _case_1_true _case_2_true) (GameplayTags|TagContainer|CompareTagAssetInterfacetoOtherTagContainers (Variables|Default|GetTC_DoubleDown_Old) (Variables|Default|GetTC_DoubleDown_T01) (Variables|Default|GetTC_DoubleDown_T02) (Variables|Default|GetInteractingPlayerPawn)))
  (bind _interactingplayerpawn (Variables|Default|GetInteractingPlayerPawn))
  (bind _returnvalue (Ability|GetAbilitySystemComponent _interactingplayerpawn))
  (bind _asfort_player_controller (Utilities|Casting|CastToFortPlayerController (Pawn|GetController (Variables|Default|GetInteractingPlayerPawn)))
    (:then
      (Inventory|GiveItemtoInventoryOwner :InventoryOwner _asfort_player_controller :ItemDefinition (Variables|Default|GetFragmentAmmoData) :NumberToGive (+ (select (and _case_0_true (Math|Random|RandomBoolwithWeight (ScalableFloat|GetValueatLevel (Variables|Default|GetRow_DoubleDown)))) (Variables|Default|GetFragmentsToGiveDoubleDown) (Variables|Default|GetFragmentsToGive)) (select (and _case_1_true (Math|Random|RandomBoolwithWeight (ScalableFloat|GetValueatLevel (Variables|Default|GetRow_DoubleDown)))) (Math|Float|Truncate (select _case_2_true (+ (ScalableFloat|GetValueatLevel (Variables|Default|GetRow_DoubleDown_T01_Fragments)) (ScalableFloat|GetValueatLevel (Variables|Default|GetRow_DoubleDown_T02_Fragments))) (ScalableFloat|GetValueatLevel (Variables|Default|GetRow_DoubleDown_T01_Fragments)))) 0)) :bNotifyPlayer false :ItemLevel -1 :PickupInstigatorHandle -1 :bUseItemPickupAnalyticEvent false :WeaponAmmoOverride -1) ; Gives Fragment Ammo to Player so he can use a Fragment Ability.
      (switch Utilities|FlowControl|Switch|SwitchonE_OutlanderFragmentTypes (Variables|Default|GetFragmentType)
        (:NewEnumerator4
          (GameplayEffects|ApplyGameplayEffectToSelf :self _returnvalue :GameplayEffectClass (Variables|Default|GetGE_OutlanderHoldingChargeFragment) :Level 0.0) ; TagsTheOutlanderAsHoldingACharge
        )
        (:NewEnumerator2
          ;[1 0.9151 0] REMOVE THIS CHECK IT IS FOR HERO LOADOUT PRESHIP
          (if (not (GameplayTags|TagContainer|CompareTagAssetInterfacetoOtherTagContainers (Variables|Default|GetTC_InTheZone_T01) _interactingplayerpawn))
            ;[1 0.9176 0] IF IN HERO LOADOUT REMOVE THIS
            (GameplayEffects|ApplyGameplayEffectToSelf :self _returnvalue :GameplayEffectClass (Variables|Default|GetGE_Outlander_LlamaFragment) :Level 0.0) ; Tags the Outlander as having a Llama Fragment.
            (GameplayEffects|ApplyGameplayEffectToSelf :self _returnvalue :GameplayEffectClass (Variables|Default|GetGE_Outlander_LlamaCooldownModifier) :Level 0.0) ; Modifies the cooldown to 2 seconds on next cast
          )))
      (Ability|SendGameplayEventtoActor (Variables|Default|GetInteractingPlayerPawn) (Variables|Default|GetEvent_FragmentCollected) (Variables|Default|GetSentFragment))
      (GameplayEffects|ApplyGameplayEffectToSelf :self _returnvalue :GameplayEffectClass (Variables|Default|GetGE_ClearFragmentCooldown) :Level 0.0)
      (return))
    (:CastFailed
      (return))))

(fn IsOutlander
  (params (Pawn :type Fort%20Pawn%20Object%20Reference))
  (Utilities|IsValid Pawn
    (:Is%20Valid
      (return (GameplayTags|HasMatchingGameplayTag Pawn "(TagName=\"Class.IsOutlander\")")))
    (:Is%20Not%20Valid
      (return false))))

(fn OnRep_BuffType
  (switch Utilities|FlowControl|Switch|SwitchonE_OutlanderFragmentTypes (Variables|Default|GetFragmentType)
    (:NewEnumerator4
      (MiniMapData|SetMiniMapIconColor (Variables|Default|GetMiniMapLoc) (Variables|Effects_MID_Colors|GetChargeFragmentColor))
      (Variables|Default|SetSentFragment (Utilities|Struct|MakeGameplayEventData :Instigator self :ContextHandle "()" :InstigatorTags "(GameplayTags=)" :TargetTags (Variables|Default|GetTC_ChargeFragmentTag) :EventMagnitude 0.0 :TargetData "()")))
    (:NewEnumerator2
      (MiniMapData|SetMiniMapIconColor (Variables|Default|GetMiniMapLoc) (Variables|Effects_MID_Colors|GetLlamaFragmentColor))
      (Variables|Default|SetSentFragment (Utilities|Struct|MakeGameplayEventData :Instigator self :ContextHandle "()" :InstigatorTags "(GameplayTags=)" :TargetTags (Variables|Default|GetTC_LlamaFragmentTag) :EventMagnitude 0.0 :TargetData "()"))))
  (CallFunction|ActivateFX))

(fn OnRep_IsOn
  (if (Variables|Default|GetIsOn)
    (Components|Activation|Activate (Variables|Default|GetP_LightOn) true)
    ;[1 1 1] Orb Audio Setup
    (Audio|SpawnSoundAttached (Variables|Default|GetOrb_Hello_Sound) (Variables|Transformation|GetRootComponent))
    (bind _returnvalue (AddComponent|NODE_AddAudioComponent-0))
    (bind _output_get (Variables|Default|SetOrbAudioComponent _returnvalue))
    (Audio|Components|Audio|SetSound _output_get (Variables|Default|GetOrb_Hover_Sound))
    (Audio|Components|Audio|FadeIn _output_get 0.5 1.0 0.0 "Linear" :!class /Script/Engine.AudioComponent)
    (else
      (Components|Activation|Deactivate (Variables|Default|GetP_LightOn))
      (Audio|SpawnSoundAttached (Variables|Default|GetOrb_Goodbye_Sound) (Variables|Transformation|GetRootComponent))
      (Audio|Components|Audio|FadeOut (Variables|Default|GetOrbAudioComponent) 1 0.0 "Linear" :!class /Script/Engine.AudioComponent)
      (Components|DestroyComponent (Variables|Default|GetOrbAudioComponent)))))

(fn OnRep_IsOpen
  (CallFunction|PlayAnimation (select (Variables|Default|GetIsOpen) (Variables|Animation|GetDeployAnim) (Variables|Animation|GetUndeployAnim))))

(fn OnRep_IsOrbTaken
  (if (Variables|Default|GetIsOrbTaken)
    (Development|SetHiddeninGame (Variables|Default|GetOrb_Mesh) true false)
    (CallFunction|PowerDown)))

(fn SetAwake
  (params (NewAwake :type Boolean))
  (Variables|Default|SetIsAwake NewAwake)
  (CallFunction|UpdateShouldTick))

(fn SetOutlander
  (params (NewOutlander :type Actor%20Object%20Reference))
  (Variables|Default|SetCurrent_Outlander NewOutlander)
  (CallFunction|UpdateShouldTick))

(fn UpdateShouldTick
  (Networking|CanExecuteCosmeticEvents
    (:True
      (Actor|Tick|SetActorTickEnabled (and (Variables|Default|GetIsAwake) (Utilities|IsValid (Variables|Default|GetCurrent_Outlander) :!class /Script/Engine.KismetSystemLibrary))))
    (:False
      (Actor|Tick|SetActorTickEnabled false))))

(fn ConstructionScript
  (Transformation|AttachTo(Deprecated) (Variables|Default|GetP_LightOn) (Variables|Default|GetContainer_SK) "light" "SnapToTarget" false)
  (Rendering|SetTranslucentSortPriority (Variables|Default|GetContainer_SK) -1))
